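# Physics asset for the NefariousJewel backpack cosmetic, giving its nine-bone dynamic chain physics bodies for secondary motion.
# A single large capsule covers the pack, and small capsules cover each chain bone.
# Constraints link the chain in sequence from the pack down to the ninth bone.
# Every joint shares the same tight swing and twist limits of a few degrees, so each link bends only slightly.
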PhysicsAsset /BRCosmetics/Accessories/FORT_Backpacks/Backpack_M_MED_NefariousJewel/Meshes/M_MED_NefariousJewel_Pack_Physics
bodies {
 dyn_chain_01 [{
   shapeType=Capsule
   length=0.922
   radius=1.3147
   shapeName=dyn_chain_01_capsule
   center=<1.8644 0.2141 -0.0131>
   rotation=<rot 0 95.6569 -88.6656>
  }]
 dyn_chain_02 [{
   shapeType=Capsule
   length=0.6888
   radius=1.1679
   shapeName=dyn_chain_02_capsule
   center=<-0.2402 0.1688 0.1168>
   rotation=<rot 0.6963 -90.906 -82.9728>
  }]
 dyn_chain_03 [{
   shapeType=Capsule
   length=0.7119
   radius=1.16
   shapeName=dyn_chain_03_capsule
   center=<-0.2042 0.0994 0.1567>
   rotation=<rot 0.0637 -93.1688 -88.5826>
  }]
 dyn_chain_04 [{
   shapeType=Capsule
   length=1.1802
   radius=1.3132
   shapeName=dyn_chain_04_capsule
   center=<-0.455 0.0056 0.2038>
   rotation=<rot 19.9811 -88.4527 -85.5725>
  }]
 dyn_chain_05 [{
   shapeType=Capsule
   length=0.2114
   radius=1.3046
   shapeName=dyn_chain_05_capsule
   center=<-0.3091 0.0297 0.2313>
   rotation=<rot 0 -86.9105 -79.7602>
  }]
 dyn_chain_06 [{
   shapeType=Capsule
   length=0.7664
   radius=1.3122
   shapeName=dyn_chain_06_capsule
   center=<-0.3761 0.2074 0.3136>
   rotation=<rot 0 -86.0121 -75.5143>
  }]
 dyn_chain_07 [{
   shapeType=Capsule
   length=0.4298
   radius=1.3308
   shapeName=dyn_chain_07_capsule
   center=<-0.315 0.2684 0.4826>
   rotation=<rot 0 -93.821 -69.7245>
  }]
 dyn_chain_08 [{shapeType=Capsule, length=1.0287, radius=1.2439, shapeName=dyn_chain_08_capsule, rotation=<rot 6.9599 -79.4011 92.2661>}]
 dyn_chain_09 [{shapeType=Capsule, length=0.505, radius=1.6574, shapeName=dyn_chain_09_capsule, rotation=<rot 0 -90 0>}]
 pack [{shapeType=Capsule, length=14.2282, radius=9.6072, center=<-4.0084 -8.6214 1.6385>, rotation=<rot -16.9231 -80.5216 -53.5828>}]
}
constraints [
 {bone1Name=dyn_chain_01, bone2Name=pack, swing1LimitDegrees=5, swing2LimitDegrees=2, twistLimitDegrees=1}
 {bone1Name=dyn_chain_02, bone2Name=dyn_chain_01, swing1LimitDegrees=5, swing2LimitDegrees=2, twistLimitDegrees=1}
 {bone1Name=dyn_chain_03, bone2Name=dyn_chain_02, swing1LimitDegrees=5, swing2LimitDegrees=2, twistLimitDegrees=1}
 {bone1Name=dyn_chain_04, bone2Name=dyn_chain_03, swing1LimitDegrees=5, swing2LimitDegrees=2, twistLimitDegrees=1}
 {bone1Name=dyn_chain_05, bone2Name=dyn_chain_04, swing1LimitDegrees=5, swing2LimitDegrees=2, twistLimitDegrees=1}
 {bone1Name=dyn_chain_06, bone2Name=dyn_chain_05, swing1LimitDegrees=5, swing2LimitDegrees=2, twistLimitDegrees=1}
 {bone1Name=dyn_chain_07, bone2Name=dyn_chain_06, swing1LimitDegrees=5, swing2LimitDegrees=2, twistLimitDegrees=1}
 {bone1Name=dyn_chain_08, bone2Name=dyn_chain_07, swing1LimitDegrees=5, swing2LimitDegrees=2, twistLimitDegrees=1}
 {bone1Name=dyn_chain_09, bone2Name=dyn_chain_08, swing1LimitDegrees=5, swing2LimitDegrees=2, twistLimitDegrees=1}
]
